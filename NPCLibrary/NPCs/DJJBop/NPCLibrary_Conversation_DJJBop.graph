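# Conversation database for the DJ Bop NPC, defining the greeting and top-level branches a player gets when talking to them.
# The entry point leads to one Speech node whose line is picked from contextual messages gated by first-ever conversation, per-match conversation count, or the player's equipped character skin.
# That node fans out through link nodes to shared quest and basic service conversations, DJ Bop's dialogue, and Season 36 Big Dill and DJ Yonder character stories.
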ConversationDatabase /NPCLibrary/NPCs/DJJBop/NPCLibrary_Conversation_DJJBop
entry_points [{tag=Conversation.Entry.DJBop, destinations [#1]}]
nodes [
 ConversationEntryPointNode_7 {
  id=#1
  class=ConversationEntryPointNode
  outputs [#2]
  properties {entryTag {tagName=Conversation.Entry.DJBop}}
 }
 Speech_C_4 {
  id=#2
  class=Speech_C
  outputs [
   #3
   #4
   #5
   #6
   #7
  ]
  properties {
   nodeName=""
   generalConfig {
    defaultMessage="I'm THE future of the Island's music scene! Just watch!"
    contextualMessages [
     {
      message="Yes, you're right! I'm THAT DJ Bop!"
      contextRequirements [{requirement {class=FortControllerRequirement_IsFirstEverConversationWithNPC}}]
      priority {curve {rowName=Tandem.Speech.Priority.FirstEverInteraction}}
     }
     {
      message="I'm THE future of the Island's music scene! Just watch!"
      contextRequirements [{requirement {class=FortControllerRequirement_CheckNPCMatchConversationCount, properties {nPCUniqueIdTag {tagName=AISpawnerData.Type.Tandem.DJBop}}}}]
      priority {curve {rowName=None}}
     }
     {
      message="Don't pay attention to gossip. I'm just an awesome DJ here to make YOU dance!"
      contextRequirements [{requirement {class=FortControllerRequirement_CheckNPCMatchConversationCount, properties {nPCUniqueIdTag {tagName=AISpawnerData.Type.Tandem.DJBop}}}}]
      priority {curve {rowName=None}}
     }
     {
      message="Ever since my apology video, people have really embraced me! It's SO nice!"
      contextRequirements [{requirement {class=FortControllerRequirement_CheckNPCMatchConversationCount, properties {nPCUniqueIdTag {tagName=AISpawnerData.Type.Tandem.DJBop}}}}]
      priority {curve {rowName=None}}
     }
     {
      message="Let's DAAAANCE!"
      contextRequirements [{requirement {class=FortControllerRequirement_CheckNPCMatchConversationCount, properties {nPCUniqueIdTag {tagName=AISpawnerData.Type.Tandem.DJBop}}}}]
      priority {curve {rowName=None}}
     }
     {
      message="You can count on me!"
      contextRequirements [{requirement {
         class=FortControllerRequirement_HasCID
         properties {allowedCIds [/BRCosmetics/Athena/Items/Cosmetics/Characters/CID_233_Athena_Commando_M_FortniteDJ]}
        }}]
      priority {curve {rowName=None}}
     }
     {
      message=You Tasty Bois are adorable. Not in the good way.
      contextRequirements [{requirement {
         class=FortControllerRequirement_HasCID
         properties {allowedCIds [
           /BRCosmetics/Athena/Items/Cosmetics/Characters/CID_614_Athena_Commando_M_MissingLink
           /BRCosmetics/Athena/Items/Cosmetics/Characters/Character_WartyBrine_Sus
          ]}
        }}]
      priority {curve {rowName=None}}
     }
    ]
   }
  }
 }
 ConversationLinkNode_4 {
  id=#7
  class=ConversationLinkNode
  properties {remoteEntryTag {tagName=Conversation.Entry.Season36.DJYonder.CharacterStories.01}}
 }
 ConversationLinkNode_3 {
  id=#6
  class=ConversationLinkNode
  properties {remoteEntryTag {tagName=Conversation.Entry.Season36.BigDill.CharacterStories.01}}
 }
 ConversationLinkNode_2 {id=#5, class=ConversationLinkNode, properties {remoteEntryTag {tagName=Conversation.Entry.DJBop.Dialogue}}}
 ConversationLinkNode_0 {id=#4, class=ConversationLinkNode, properties {remoteEntryTag {tagName=Conversation.Entry.BasicServices}}}
 ConversationLinkNode_1 {id=#3, class=ConversationLinkNode, properties {remoteEntryTag {tagName=Conversation.Entry.NPCLibrary.Quests}}}
]
