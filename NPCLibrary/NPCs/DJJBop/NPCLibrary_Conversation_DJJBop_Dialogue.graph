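# Conversation database for the DJ Bop NPC's "Ask about..." dialogue branch, where the DJ gives in-character banter.
# An entry tag leads to a greeting that offers two topics, The Tasty Bois and "That" Incident, plus a Back option.
# Each topic plays a short linear chain of speech lines, and both chains converge on a shared ReturnToStart node labeled BACK.
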ConversationDatabase /NPCLibrary/NPCs/DJJBop/NPCLibrary_Conversation_DJJBop_Dialogue
entry_points [{tag=Conversation.Entry.DJBop.Dialogue, destinations [#1]}]
nodes [
 ConversationEntryPointNode_0 {
  id=#1
  class=ConversationEntryPointNode
  outputs [#2]
  properties {entryTag {tagName=Conversation.Entry.DJBop.Dialogue}}
 }
 Speech_C_10 {
  id=#2
  class=Speech_C
  outputs [#3, #4, #5]
  properties {nodeName="", softTaskIcon=&T_UI_NPC_question, generalConfig {defaultMessage=What's UP??}}
  sub_nodes [ConversationChoiceNode_9 {class=ConversationChoiceNode, properties {defaultChoiceDisplayText=Ask about...}}]
 }
 Speech_C_13 {
  id=#5
  class=Speech_C
  outputs [#6]
  properties {
   nodeName=""
   softTaskIcon=&T_UI_NPC_question
   generalConfig {defaultMessage=They're the new kid on the block and they're ALREADY old news.}
  }
  sub_nodes [ConversationChoiceNode_13 {class=ConversationChoiceNode, properties {defaultChoiceDisplayText=The Tasty Bois}}]
 }
 Speech_C_14 {
  id=#6
  class=Speech_C
  outputs [#7]
  properties {
   nodeName=""
   softTaskIcon=&T_UI_NPC_question
   generalConfig {defaultMessage="Tasty Bois? More like Expired Bois. Their beats aren't sick, they'll just MAKE YOU sick. HA!"}
  }
 }
 ReturnToStart_C_1 {
  id=#7
  class=ReturnToStart_C
  properties {softTaskIcon=&T_UI_NPC_Back}
  sub_nodes [ConversationChoiceNode_8 {class=ConversationChoiceNode, properties {defaultChoiceDisplayText=BACK}}]
 }
 Speech_C_1 {
  id=#4
  class=Speech_C
  outputs [#8]
  properties {
   nodeName=""
   softTaskIcon=&T_UI_NPC_question
   generalConfig {defaultMessage="Everything I'm going to say, I SAID in my apology video."}
  }
  sub_nodes [ConversationChoiceNode_21 {class=ConversationChoiceNode, properties {defaultChoiceDisplayText="\"That\" Incident"}}]
 }
 Speech_C_2 {
  id=#8
  class=Speech_C
  outputs [#9]
  properties {
   nodeName=""
   softTaskIcon=&T_UI_NPC_question
   generalConfig {defaultMessage="It was a simple little mistake and it's OVER now!"}
  }
 }
 Speech_C_15 {
  id=#9
  class=Speech_C
  outputs [#7]
  properties {
   nodeName=""
   softTaskIcon=&T_UI_NPC_question
   generalConfig {defaultMessage="We have better things to do than gossip... like DANCING!"}
  }
 }
 Back_C_0 {
  id=#3
  class=Back_C
  sub_nodes [ConversationChoiceNode_20 {class=ConversationChoiceNode, properties {defaultChoiceDisplayText=BACK}}]
 }
]
